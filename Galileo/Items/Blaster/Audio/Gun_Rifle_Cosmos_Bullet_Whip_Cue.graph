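# Sound cue for the Cosmos blaster rifle's bullet whiz-by, played as projectiles pass.
# A random node picks one of four equally weighted whiz-by wave variations at half volume.
# An attenuation override turns off distance volume falloff and sets a wide stereo spread.
# It routes to the ranged weapon projectile effects sound class and uses the Athena bullet whiz-by concurrency rules to limit overlapping instances.
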SoundCue /Galileo/Items/Blaster/Audio/Gun_Rifle_Cosmos_Bullet_Whip_Cue
root=Random_2
nodes [
 {id=Random_2, children [WavePlayer_6,WavePlayer_7,WavePlayer_8,WavePlayer_9], properties {weights [1,1,1,1]}}
 {id=WavePlayer_6, properties {soundWaveAssetPtr=&/Gun_Rifle_Cosmo_Bullet_WhizBy_01}}
 {id=WavePlayer_7, properties {soundWaveAssetPtr=&/Gun_Rifle_Cosmo_Bullet_WhizBy_02}}
 {id=WavePlayer_8, properties {soundWaveAssetPtr=&/Gun_Rifle_Cosmo_Bullet_WhizBy_03}}
 {id=WavePlayer_9, properties {soundWaveAssetPtr=&/Gun_Rifle_Cosmo_Bullet_WhizBy_04}}
]
properties {
 bOverrideAttenuation=true
 maxDistance=2097152
 soundClassObject=&Fort_WeaponRanged_Projectile_Effects
 volumeMultiplier=0.5
 attenuationOverrides {bAttenuate=false, stereoSpread=5000}
 concurrencySet [/Game/Sounds/Concurrency_Rules/Athena_BulletWhizBy_Conc]
}
