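; Blueprint projectile for the Galileo Cosmos AR blaster bullet, with gravity set at construction and muzzle origin and start time fed to its mesh material.
; Stop and capsule-hit events run duplicate bStopped-guarded logic: disable collision, play impact sound, hide mesh, and request render-gated decals off mobile for non-pawn hits.
; Any touch overlap plays water sound and FX, then destroys the actor, without filtering on its water casts.
; Predicted-to-real replacement carries over the stopped flag.
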(event EventBeginPlay
  (bind _projectilemesh (Variables|Default|GetProjectileMesh))
  (Variables|Default|SetBeginTime (Utilities|Time|GetTimeSeconds))
  (|CallParentFunction)
  (Networking|CanExecuteCosmeticEvents
    (:True
      (bind _asfort_player_pawn (Utilities|Casting|CastToFortPlayerPawn (Variables|Actor|GetInstigator))
        (:then
          (Rendering|Material|SetVectorParameterValueonMaterials _projectilemesh "BulletOrigin" (Effects|GetMuzzleLocation (Variables|/Script/FortniteGame.FortPawn|GetCurrentWeapon _asfort_player_pawn) -1))
          (Rendering|Material|SetScalarParameterValueonMaterials _projectilemesh "StartTime" (Utilities|Time|GetGameTimeinSeconds)))
        (:CastFailed
          (Development|PrintString "CAST FAILED"))))))

(event EventOnStop
  (params (Hit :type Hit%20Result%20Structure%20%28by%20ref%29))
  (bind (_bblockinghit _binitialoverlap _time _distance _location _impactpoint _normal _impactnormal _physmat _hitactor) (Collision|BreakHitResult Hit))
  (bind (_asfort_pawn _bsuccess) (Utilities|Casting|CastToFortPawn _hitactor))
  ;[1 1 1] Projectile Impact / Stop
  (Variables|Default|SetHit Hit)
  (|CallParentFunction Hit)
  (if (not (Variables|Default|GetbStopped))
    (Variables|Default|SetbHitPawn _bsuccess)
    (Variables|Default|SetImpactLocation _location)
    (Variables|Default|SetImpactNormal _impactnormal)
    (Variables|Default|SetbStopped true)
    (Utilities|FlowControl|Sequence
      (:then_0
        (Class|BulletWhipTrackerComponentBase|Disable (Variables|Visual|GetBulletWhipTrackerComponent))
        (Collision|SetActorEnableCollision false)
        (Collision|SetCollisionEnabled (Variables|Collision|GetCapsuleComponent) "NoCollision"))
      (:then_1
        (Networking|CanExecuteCosmeticEvents
          (:True
            (Audio|PlaySoundatLocation (select (Variables|Default|GetbHitPawn) (Variables|Default|GetSniper_Rifle_PlayerImpactSound) (Variables|Default|GetSniper_Rifle_SurfaceImpactSound)) (Transformation|GetActorLocation))
            (if (not (or (Fortnite|IsMobilePlatform) (Variables|Default|GetbHitPawn)))
              (CallFunction|CheckforCreateDecal)
              (Utilities|FlowControl|Delay 0.02))
            (Rendering|SetVisibility (Variables|Default|GetProjectileMesh) false false)
            (Effects|Components|ParticleSystem|SetBoolParameter (Variables|Default|GetNS_Galileo_Bun_Prj_Light) "ProjCollided" true)))))))

(event OnComponentHit(CapsuleComponent)
  (params (HitComponent :type Primitive%20Component%20Object%20Reference) (OtherActor :type Actor%20Object%20Reference) (OtherComp :type Primitive%20Component%20Object%20Reference) (NormalImpulse :type Vector :default "0, 0, 0") (Hit :type Hit%20Result%20Structure%20%28by%20ref%29))
  (bind (_bblockinghit _binitialoverlap _time _distance _location _impactpoint _normal _impactnormal _physmat _hitactor) (Collision|BreakHitResult Hit))
  (bind (_asfort_pawn _bsuccess) (Utilities|Casting|CastToFortPawn _hitactor))
  ;[1 1 1] Projectile Impact / Stop
  (Variables|Default|SetHit Hit)
  (if (not (Variables|Default|GetbStopped))
    (Variables|Default|SetbHitPawn _bsuccess)
    (Variables|Default|SetImpactLocation _location)
    (Variables|Default|SetImpactNormal _impactnormal)
    (Variables|Default|SetbStopped true)
    (Utilities|FlowControl|Sequence
      (:then_0
        (Class|BulletWhipTrackerComponentBase|Disable (Variables|Visual|GetBulletWhipTrackerComponent))
        (Collision|SetActorEnableCollision false)
        (Collision|SetCollisionEnabled (Variables|Collision|GetCapsuleComponent) "NoCollision"))
      (:then_1
        (Networking|CanExecuteCosmeticEvents
          (:True
            (Audio|PlaySoundatLocation (select (Variables|Default|GetbHitPawn) (Variables|Default|GetSniper_Rifle_PlayerImpactSound) (Variables|Default|GetSniper_Rifle_SurfaceImpactSound)) (Transformation|GetActorLocation))
            (if (not (or (Fortnite|IsMobilePlatform) (Variables|Default|GetbHitPawn)))
              (CallFunction|CheckforCreateDecal)
              (Utilities|FlowControl|Delay 0.02))
            (Rendering|SetVisibility (Variables|Default|GetProjectileMesh) false false)
            (Effects|Components|ParticleSystem|SetBoolParameter (Variables|Default|GetNS_Galileo_Bun_Prj_Light) "ProjCollided" true)))))))

(event Custom|CheckForCreateDecal
  (bind _hit (Variables|Default|GetHit))
  (bind (_bblockinghit _binitialoverlap _time _distance _location _impactpoint _normal _impactnormal _physmat _hitactor _hitcomponent) (Collision|BreakHitResult _hit))
  ;[1 1 1] Spawn Decal if either projectile or weapon was rendered recently
  (if (not (Rendering|WasActorRecentlyRendered 0.2))
    (Utilities|IsValid _hitactor
      (:Is%20Valid
        (if (not (Rendering|WasActorRecentlyRendered _hitactor 0.2))
          (Utilities|IsValid _hitcomponent
            (:Is%20Valid
              (if (Rendering|WasComponentRecentlyRendered _hitcomponent 0.2)
                _))))
        (bind _returnvalue (Rendering|Decal|SpawnDecalatLocation "/Game/Weapons/Materials/MD_Decal_BulletHole_Lifetime_Inst1.MD_Decal_BulletHole_Lifetime_Inst1" (Variables|Default|GetDecalSize) (Variables|Default|GetImpactLocation) (Math|Conversions|RotationFromXVector (* (Variables|Default|GetImpactNormal) -1.0)) 4.5))
        (Rendering|Components|Decal|SetFadeOut _returnvalue 0.0 4.5 true)
        (Development|PrintString "RecentlyRendered"))
      (:Is%20Not%20Valid
        _))))

(event EventOnTouched
  (params (OtherActor :type Actor%20Object%20Reference) (OtherComp :type Primitive%20Component%20Object%20Reference) (HitResult :type Hit%20Result%20Structure%20%28by%20ref%29) (bIsOverlap :type Boolean))
  (bind _waterhit (Variables|Default|GetWaterHit))
  (bind (_bblockinghit _binitialoverlap _time _distance _location _impactpoint _normal _impactnormal _physmat _hitactor) (Collision|BreakHitResult _waterhit))
  (bind _waterinteractioncomponent (Variables|Water|GetWaterInteractionComponent))
  (bind (_waterplanelocation _waterplanenormal _watersurfaceposition _waterdepth _waterbodyidx) (FX|GetLastWaterSurfaceInfo _waterinteractioncomponent :!class /Script/FortniteGame.FortWaterInteractionComponent))
  (bind _returnvalue (Game|GetInstigator))
  (bind _returnvalue_2 (Pawn|IsLocallyControlled _returnvalue))
  (bind (_asfort_athena_water _bsuccess) (Utilities|Casting|CastToFortAthenaWater _hitactor))
  (bind (_asfort_water_body_actor _bsuccess_2) (Utilities|Casting|CastToFortWaterBodyActor _hitactor))
  (bind _returnvalue_3 (or _bsuccess_2 _bsuccess))
  (Networking|CanExecuteCosmeticEvents
    (:True
      (if bIsOverlap
        ;[0 0.298 1] Water Impact Projectile
        (Variables|Default|SetWaterHit HitResult)
        (Rendering|Material|SetScalarParameterValueonMaterials (Variables|Default|GetProjectileMesh) "Alpha" 0.0)
        (if (and _returnvalue_3 _returnvalue_2)
          _
          (elif (and _returnvalue_3 (not _returnvalue_2))
            _))
        (Utilities|FlowControl|Sequence
          (:then_0
            (CallFunction|PlayImpactSFX false true)
            (bind _returnvalue_4 (Effects|SpawnEmitteratLocation (Variables|Default|GetWaterImpactFX) _location (Math|Conversions|RotationFromXVector _waterplanenormal)))
            (Effects|Components|ParticleSystem|SetFloatParameter _returnvalue_4 "WaterDepth" _waterdepth)
            (Effects|Components|ParticleSystem|SetFloatParameter _returnvalue_4 "WaterBodyIndex" (Math|Conversions|ToFloat(Integer) _waterbodyidx)))
          (:then_1
            (Actor|DestroyActor)
            (Rendering|SetVisibility (Variables|Default|GetProjectileMesh) false false)
            (Components|Activation|Deactivate (Variables|Default|GetNS_Galileo_Bun_Prj_Light))
            (Development|PrintString "delete mesh")))))))

(event EventReplacedByRealProjectile
  (params (RealProjectile :type Fort%20Projectile%20Athena%20Object%20Reference))
  ;[1 1 1] We don't have anything to do here, just go away
  (Variables|Default|SetbStopped true))

(event EventReplacedPredictedProjectile
  (params (PredictedProjectile :type Fort%20Projectile%20Base%20Object%20Reference))
  ;[1 1 1] We may just want to hide the tracer on the real projectile if the dummy has already sold the effect
  (Utilities|FlowControl|Sequence
    (:then_0
      (bind _asb_prj_bullet_cosmos_ar (Utilities|Casting|CastToB_Prj_Bullet_Cosmos_AR PredictedProjectile))
      (if (Variables|Default|GetbStopped _asb_prj_bullet_cosmos_ar)
        (Variables|Default|SetbStopped true)))))

(fn PlayImpactSFX
  (params (isPlayer :type Boolean) (isWater_ :type Boolean))
  (bind _returnvalue (Transformation|GetActorLocation))
  (if isWater_
    (Audio|PlaySoundatLocation (Variables|Default|GetSniper_Rifle_WaterImpactSound) _returnvalue)
    (Variables|Default|SetbHasPlayedWaterSound true)
    (elif (Variables|Default|GetbHasPlayedWaterSound)
      _
      (else
        (Audio|PlaySoundatLocation (select isPlayer (Variables|Default|GetSniper_Rifle_PlayerImpactSound) (Variables|Default|GetSniper_Rifle_SurfaceImpactSound)) _returnvalue)))))

(fn ConstructionScript
  (Projectile|SetGravityScale (ScalableFloat|GetValueatLevel (Variables|Default|GetProjectileGravityScale))))
